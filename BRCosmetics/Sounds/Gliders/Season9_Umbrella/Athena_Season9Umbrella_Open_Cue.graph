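# Sound cue for the open sound of the Season 9 umbrella glider cosmetic.
# A Random node picks one of three equally weighted umbrella open wave variations.
# Shared glider open/close attenuation and concurrency, a glider-open property sheet and the Fort_Gliders_Land sound class govern playback, with slightly reduced volume.
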SoundCue /BRCosmetics/Sounds/Gliders/Season9_Umbrella/Athena_Season9Umbrella_Open_Cue
root=Random_1
nodes [
 {id=Random_1, children [WavePlayer_1,WavePlayer_2,WavePlayer_4], properties {weights [1,1,1]}}
 {id=WavePlayer_1, properties {soundWaveAssetPtr=&/SFX_Umbrella_S9_Open_01}}
 {id=WavePlayer_2, properties {soundWaveAssetPtr=&/SFX_Umbrella_S9_Open_02}}
 {id=WavePlayer_4, properties {soundWaveAssetPtr=&/SFX_Umbrella_S9_Open_03}}
]
properties {
 attenuationSettings=/Game/Sounds/Attenuation/Fort_Player_Glider_OpenClose_Athena_Att
 audioPropertiesSheet=/BRCosmetics/Sounds/Settings/PropertySheets/Glider/APS_Cosmetic_Glider_Open
 maxDistance=10000
 soundClassObject=&Fort_Gliders_Land
 volumeMultiplier=0.7
 concurrencySet [&Athena_GliderOpenClose_Conc]
}
